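# Slick02 SE vehicle wheel material that adds a glowing, color-tintable ring with fake depth to the simple wheel parent.
# The ring texture's green channel is sampled through chained BumpOffsets and summed with falling weights, giving layered parallax emissive on black.
# A UDIM mask blends it over MF_WheelParent_Simple, before MF_Vehicle_FNPostFX.
# Ring_LOD drives a switch for an alternate mask built from inverted vertex red and texture blue.
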Material /VehicleCosmetics/Wheels/Slick02/Materials/M_Wheel_Slick02_SE
graph {
 properties {
  bAutomaticallySetUsageInEditor=false
  bUseMaterialAttributes=true
  bUsedWithMeshDeformer=true
  bUsedWithSkeletalMesh=true
  usedShadingModels=Default Lit
 }
 nodes [
  n1 = NamedRerouteUsage()
  n2 = NamedRerouteUsage()
  n3 = TextureSample(UVs=@n1, Tex=@n2) {samplerSource=SSM_Wrap_WorldGroupSettings, samplerType=SAMPLERTYPE_LinearColor}
  n4 = NamedRerouteUsage()
  n5 = NamedRerouteUsage()
  n6 = BumpOffset(Coordinate=@n1, Height=@n5)
  n7 = TextureSample(UVs=@n6, Tex=@n2) {samplerSource=SSM_Wrap_WorldGroupSettings, samplerType=SAMPLERTYPE_LinearColor}
  n8 = Add(A=LinearInterpolate(A=@n3.G, B=@n3.G, Alpha=@n4), B=Multiply(A=LinearInterpolate(A=@n7.G, B=@n7.G, Alpha=@n4)) {constB=0.6})
  n9 = BumpOffset(Coordinate=@n6, Height=@n5)
  n10 = TextureSample(UVs=@n9, Tex=@n2) {samplerSource=SSM_Wrap_WorldGroupSettings, samplerType=SAMPLERTYPE_LinearColor}
  n11 = TextureSample(UVs=BumpOffset(Coordinate=@n9, Height=@n5), Tex=@n2) {samplerSource=SSM_Wrap_WorldGroupSettings, samplerType=SAMPLERTYPE_LinearColor}
  n12 = Add(
   A=Multiply(A=LinearInterpolate(A=@n10.G, B=@n10.G, Alpha=@n4)) {constB=0.4}
   B=Multiply(A=LinearInterpolate(A=@n11.G, B=@n11.G, Alpha=@n4)) {constB=0.2}
  )
  n13 = VectorParameter() {group=00 Base, parameterName=Custom Color, sortPriority=0, defaultValue=<rgba 0.0409 0.7011 0.8469 1>}
  n14 = ScalarParameter() {defaultValue=2, group=00 Base, parameterName=Emissive Strength}
  n15 = Constant()
  n16 = Constant() {r=1}
  n17 = SetMaterialAttributes!(
   MaterialAttributes
   Emissive Color=Multiply(A=Add(A=@n8, B=@n12), B=Multiply(A=@n13.RGB, B=@n14))
   Base Color=@n15
   Metallic=@n16
   Specular=@n15
   Roughness=@n16
  ) {attributeSetTypes [
    #1
    #2
    #3
    #4
    #5
   ]}
  n18 = MaterialFunctionCall!(UDimm_UVLocation=Constant2Vector() {g=1}) {materialFunction=/VehicleCosmetics/Materials/ExtraFunctions/MF_UDimmMask}
  n19 = NamedRerouteUsage()
  n20 = NamedRerouteUsage()
  MaterialFunctionCall!(In=BlendMaterialAttributes(
    A=MaterialFunctionCall() {materialFunction=&MF_WheelParent_Simple}.MaterialAttributes
    B=@n17
    Alpha=LinearInterpolate(A=@n18.Mask, B=@n19, Alpha=@n20)
   )) {materialFunction=/VehicleCosmetics/Materials/MaterialFunctions/MF_Vehicle_FNPostFX}
  n21 = VertexColor()
  NamedRerouteDeclaration(None=Multiply(A=OneMinus(None=@n21.R), B=@n3.B)) {name=Ring_LODMask, nodeColor=<rgba 0 1 0.8125 1>}
  n22 = ScalarParameter() {group=00 Base, parameterName=Ring_LOD}
  NamedRerouteDeclaration(None=Saturate(None=Ceil(None=@n22))) {name=LOD_Switch, nodeColor=<rgba 1 0.2344 0 1>}
  NamedRerouteDeclaration(None=TextureCoordinate()) {name=Coords_1, nodeColor=<rgba 1 0.8203 0 1>}
  n23 = ScalarParameter() {defaultValue=-0.75, group=00 Base, parameterName=LayerDepth, sortPriority=7}
  NamedRerouteDeclaration(None=LinearInterpolate(A=@n23, B=Multiply(A=@n23) {constB=-0.3}, Alpha=NamedRerouteUsage())) {name=LayerDepth_1, nodeColor=<rgba 0.5 0 1 1>}
  NamedRerouteDeclaration(None=TextureObjectParameter() {group=00 Base, parameterName=Ring, sortPriority=3, texture=/VehicleCosmetics/Wheels/Slick02/Textures/T_Slick_02_SE}) {name=Ring, nodeColor=<rgba 0.0312 1 0 1>}
 ]
}
